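; Calendar prop building actor (Apollo Donut) that shows a date-driven state through its mesh materials.
; On BeginPlay it reads the current UTC date and computes a day index as (month minus 3) times 31 plus day, ignoring the year.
; It writes that index as a float to a scalar parameter, named by a variable, on the mesh's materials.
; The index assumes 31-day months counted from March, and the construction script is empty.
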(event EventBeginPlay
  (bind (_year _month _day) (Math|DateTime|BreakDateTime (Math|DateTime|GetDate (Math|DateTime|UTCNow))))
  (Rendering|Material|SetScalarParameterValueonMaterials (Variables|Default|GetMesh) (Variables|Default|GetParameterName) (Math|Conversions|ToFloat(Integer) (+ (* (- _month 3) 31) _day))))

(fn ConstructionScript)
